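# Translucent UI material for the emote highlight button background, drawing a faint box fill with an animated outline highlight.
# MF_UI_Box supplies the button fill and outline, and a second, rotated box forms a soft-edged mask sized by MaskX and MaskY.
# Time scaled by RotationSpeed samples a curve atlas row in reverse to drive the mask rotation, with the masked outline added to the Opacity-scaled fill.
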Material /FortUILibrary/Foundation/Buttons/Materials/EmoteHighlightButton/M_UI_EmotesHighlightBackground
graph {
 properties {
  bUsedWithStaticMesh=false
  blendMode=BLEND_Translucent
  materialDomain=MD_UI
  pixelDepthOffsetMode=PDOM_Legacy
  usedShadingModels=Default Lit
 }
 nodes [
  Add(
   A=Multiply(A=NamedRerouteUsage(), B=ScalarParameter() {defaultValue=0.3, parameterName=Opacity})
   B=Multiply(A=NamedRerouteUsage(), B=NamedRerouteUsage())
  )
  Constant3Vector() {constant=<rgba 1 1 1 0>}
  n1 = MaterialFunctionCall(Size=Constant() {r=1}, OutlineThickness=Constant() {r=0.1}) {materialFunction=&MF_UI_Box}
  NamedRerouteDeclaration(None=@n1.Fill) {name=ButtonShape, nodeColor=<rgba 0 0 0 1>}
  NamedRerouteDeclaration(None=OneMinus(None=SmoothStep(Value=@n1.SDFOutline) {constMax=0.05})) {name=Outline, nodeColor=<rgba 0 0 0 1>}
  n2 = MaterialFunctionCall(
   UVs=MaterialFunctionCall(UVs=TextureCoordinate(), "Rotation Angle (0-1)"=NamedRerouteUsage()) {materialFunction=&CustomRotator}.Rotated Values
   Size=AppendVector(A=ScalarParameter() {defaultValue=0.1944, parameterName=MaskX}, B=ScalarParameter() {defaultValue=2, parameterName=MaskY})
  ) {materialFunction=&MF_UI_Box}
  NamedRerouteDeclaration(None=OneMinus(None=SmoothStep(Value=@n2.SDF) {constMax=0.2688})) {name=Mask, nodeColor=<rgba 0 0 0 1>}
  n3 = Multiply(A=Time(), B=ScalarParameter() {defaultValue=0.3079, parameterName=RotationSpeed})
  n4 = CurveAtlasRowParameter(CurveTime=Frac(None=Multiply(A=@n3) {constB=-1})) {
   atlas=&/CA_EmotesHighlight_Background_Atlas
   curve=&/C_EmotesHighlight_Background
   parameterName=RotationCurve
  }
  NamedRerouteDeclaration(None=@n4.R) {name=RotationCurve, nodeColor=<rgba 0 0 0 1>}
 ]
}
